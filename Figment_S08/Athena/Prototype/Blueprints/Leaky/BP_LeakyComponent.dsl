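; BP_LeakyComponent is a prototype Blueprint, named as a component, that relays damage on props for a "Leaky" feature.
; Three custom events, DamageProp0 to DamageProp2, each take an integer damage amount and call the OnDamageAdded dispatcher with it and a fixed index matching the prop number.
; It applies no damage logic itself, leaving the response to listeners, and its BeginPlay and Tick events are empty.
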(event EventBeginPlay)

(event EventTick
  (params (DeltaSeconds :type Float%20%28single-precision%29)))

(event Custom|DamageProp0
  (params (DamageApplied :type Integer))
  (Default|CallOnDamageAdded DamageApplied))

(event Custom|DamageProp1
  (params (DamageApplied :type Integer))
  (Default|CallOnDamageAdded DamageApplied 1))

(event Custom|DamageProp2
  (params (DamageApplied :type Integer))
  (Default|CallOnDamageAdded DamageApplied 2))
