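# Translucent unlit material for an iris mesh in the Leaky prototype effects, drawing a glowing, tintable flare disc.
# The red channel of a spiky flare texture blends FlareColor toward an intensified version, and a UV-centered SphereMask shapes the result.
# That mask also scales with OverallOpacity for transparency.
# A feature-level switch uses DeriveHDRfromLDR by default and a simple HDRIntensity multiply on ES2 and ES3.1.
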Material /Figment_S08/Athena/Prototype/Blueprints/Leaky/FX/M_Leaky_IrisMesh
graph {
 properties {blendMode=BLEND_Translucent, pixelDepthOffsetMode=PDOM_Legacy, shadingModel=MSM_Unlit, usedShadingModels=Unlit}
 nodes [
  n1 = VectorParameter() {parameterName=FlareColor, defaultValue=<rgba 0 0.6366 1 1>}
  n2 = ScalarParameter() {defaultValue=10, parameterName=Intensity}
  n3 = SphereMask(
   A=TextureCoordinate()
   B=Constant() {r=0.5}
   Radius=ScalarParameter() {defaultValue=0.5, parameterName=Radius_b}
   Hardness=ScalarParameter() {defaultValue=0.5, parameterName=Hardness_b}
  )
  n4 = Multiply(
   A=LinearInterpolate(A=@n1.RGB, B=Multiply(A=@n1.RGB, B=@n2), Alpha=TextureSample() {texture=/Game/Effects/Fort_Effects/Textures/Flares/T_Flare_Spikeball}.R)
   B=@n3
  )
  n5 = ScalarParameter() {defaultValue=4, parameterName=HDRIntensity}
  n6 = MaterialFunctionCall(LDR Input=Clamp(None=@n4), HDR Tint=@n4, HDR Intensity Multiplier=@n5) {materialFunction=&DeriveHDRfromLDR}
  n7 = Multiply(A=@n4, B=@n5)
  FeatureLevelSwitch(Default=@n6.RemappedRange, ES2=@n7, ES3_1=@n7, SM4_REMOVED=@n6.RemappedRange)
  Multiply(A=@n3, B=ScalarParameter() {defaultValue=1, parameterName=OverallOpacity})
 ]
}
