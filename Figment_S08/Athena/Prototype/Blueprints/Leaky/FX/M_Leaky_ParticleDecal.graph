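# Translucent unlit particle-sprite material for Leaky FX, drawing an animated iridescent pattern as a decal-like glow on nearby surfaces.
# Two scrolling flow maps distort three offset gradient samples into split RGB lines, shaped by a bubble mask and tinted by particle color.
# Scene-normal Fresnel, with a dynamic-parameter exponent, adds edge glow and joins depth fades and a radial mask in opacity.
# Sprites collapse toward their center near the camera; low-end paths use constant fallbacks.
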Material /Figment_S08/Athena/Prototype/Blueprints/Leaky/FX/M_Leaky_ParticleDecal
graph {
 properties {
  bUsedWithParticleSprites=true
  blendMode=BLEND_Translucent
  pixelDepthOffsetMode=PDOM_Legacy
  shadingModel=MSM_Unlit
  usedShadingModels=Unlit
 }
 nodes [
  n1 = Multiply(A=TextureCoordinate(), B=ScalarParameter() {defaultValue=25, parameterName=Base Tile})
  n2 = Time()
  n3 = ScalarParameter() {defaultValue=0.05, parameterName=DistortA-Speed}
  n4 = Add(
   A=Multiply(A=TextureCoordinate(), B=ScalarParameter() {defaultValue=2, parameterName=DistortA-Tile})
   B=Multiply(A=Multiply(A=@n2, B=@n3), B=Constant2Vector() {g=0.25, r=1})
  )
  n5 = TextureCoordinate()
  n6 = ScalarParameter() {defaultValue=4, parameterName=DistortB-Tile}
  n7 = Multiply(A=Time(), B=ScalarParameter() {defaultValue=-0.1, parameterName=DistortB-Speed})
  n8 = Constant2Vector() {g=1, r=0.25}
  n9 = Add(A=ScalarParameter() {defaultValue=0.33, parameterName=DistortB-Offset}, B=Add(A=Multiply(A=@n5, B=@n6), B=Multiply(A=@n7, B=@n8)))
  n10 = Add(
   A=ComponentMask(None=TextureSample(UVs=@n4) {samplerType=SAMPLERTYPE_LinearColor, texture=/Figment_S08/Environments/Sets/DigSite/Textures/T_Push_Flow}.RGB) {g=true, r=true}
   B=ComponentMask(None=TextureSample(UVs=@n9) {samplerType=SAMPLERTYPE_LinearColor, texture=&T_Push_Flow}.RGB) {g=true, r=true}
  )
  n11 = Multiply(A=ScalarParameter() {defaultValue=25, parameterName=Distort-BaseAmt}, B=Subtract(A=Normalize(VectorInput=@n10)) {constB=0.5})
  n12 = Panner(Coordinate=Add(A=@n1, B=@n11)) {speedY=-0.125}
  n13 = ScalarParameter() {defaultValue=1.02, parameterName=Distort-GreenAmt}
  n14 = Panner(Coordinate=Add(A=Multiply(A=@n11, B=@n13), B=@n1)) {speedY=-0.125}
  n15 = ScalarParameter() {defaultValue=1.04, parameterName=Distort-BlueAmt}
  n16 = Panner(Coordinate=Add(A=Multiply(A=@n11, B=@n15), B=@n1)) {speedY=-0.125}
  n17 = Saturate(None=TextureSample(UVs=@n4) {samplerType=SAMPLERTYPE_LinearColor, texture=/Figment_S08/Environments/Sets/DigSite/Textures/T_Push_Mask}.R)
  n18 = Power(
   Base=MaterialFunctionCall(
    X=TextureSample(UVs=@n12) {samplerType=SAMPLERTYPE_LinearColor, texture=/Game/Weapons/WeaponSkins/Textures/T_VertGradient}.R
    Y=TextureSample(UVs=@n14) {samplerType=SAMPLERTYPE_LinearColor, texture=&T_VertGradient}.R
    Z=TextureSample(UVs=@n16) {samplerType=SAMPLERTYPE_LinearColor, texture=&T_VertGradient}.R
   ) {materialFunction=&MakeFloat3}.Result
   Exp=LinearInterpolate(
    A=ScalarParameter() {defaultValue=0.25, parameterName=LinesPower-Low}
    B=ScalarParameter() {defaultValue=10, parameterName=LinesPower}
    Alpha=@n17
   )
  )
  n19 = Saturate(None=Add(A=@n17, B=ScalarParameter() {defaultValue=0.05, parameterName=BubbleDarkAmt}))
  n20 = Multiply(
   A=Desaturation(None=Multiply(A=@n18, B=@n19), Fraction=ScalarParameter() {defaultValue=-2, parameterName=Desaturation})
   B=ScalarParameter() {defaultValue=3, parameterName=Magnetic Intensity}
  )
  n21 = ParticleColor()
  n22 = ScalarParameter() {defaultValue=0.6, parameterName=Lack of Particle Influence}
  n23 = SceneTexture() {sceneTextureId=PPI_WorldNormal}
  n24 = Fresnel(
   ExponentIn=DynamicParameter() {paramNames [FresnelExponent, Param2, Param3, Param4]}.FresnelExponent
   Normal=MaterialFunctionCall(Normal=ComponentMask(None=@n23.Color) {b=true, g=true, r=true}, Flatness=Constant() {r=0.2}) {materialFunction=&FlattenNormal}.Result
  ) {baseReflectFraction=0, exponent=1}
  n25 = Constant() {r=0.3}
  n26 = ShadingPathSwitch(Default=@n24, Deferred=@n24, Forward=@n25, Mobile=@n24)
  n27 = FeatureLevelSwitch(Default=@n26, ES2=@n25, ES3_1=@n25, SM4_REMOVED=@n26)
  Add(
   A=Multiply(A=@n20, B=LinearInterpolate(A=@n21.RGB, Alpha=@n22) {constAlpha=0.6})
   B=Multiply(A=Multiply(A=@n21.RGB, B=@n27), B=ScalarParameter() {defaultValue=5, parameterName=WorldNormalIntensity}) {constB=5, desc=Fresnel edge color}
  )
  n28 = ParticlePositionWS()
  n29 = ObjectPositionWS()
  n30 = FeatureLevelSwitch(Default=@n28, ES2=@n29, ES3_1=@n29, SM4_REMOVED=@n28, SM5=@n28, SM6=@n28) {desc="Please fix particle position on mobile :("}
  n31 = Distance(A=@n30, B=CameraPositionWS())
  n32 = ScalarParameter() {defaultValue=250, parameterName=NearFadeDistance}
  LinearInterpolate(A=Subtract(A=@n30, B=WorldPosition().XYZ), Alpha=Clamp(None=Divide(A=@n31, B=@n32))) {constB=0}
  n33 = Constant() {r=0.5}
  n34 = SphereMask(A=TextureCoordinate(), B=@n33, Radius=@n33) {hardnessPercent=0}
  n35 = ScalarParameter() {defaultValue=2, parameterName=SphereMaskPower}
  n36 = ScalarParameter() {defaultValue=1, parameterName=TextureOpacityAdd}
  n37 = DepthFade(FadeDistance=ScalarParameter() {defaultValue=44, parameterName=VisibleDepthDistance}) {fadeDistanceDefault=80}
  n38 = DepthFade(Opacity=Add(A=Multiply(A=@n20, B=@n36), B=OneMinus(None=@n37)), FadeDistance=ScalarParameter() {defaultValue=20, parameterName=DepthFade}) {fadeDistanceDefault=15}
  n39 = Constant() {r=0.75}
  Multiply(
   A=Multiply(A=Multiply(A=@n21.A, B=@n27) {constA=1}, B=Power(Base=@n34, Exp=@n35))
   B=FeatureLevelSwitch(Default=@n38, ES2=@n39, ES3_1=@n39, SM4_REMOVED=@n38, SM5=@n38, SM6=@n38)
  )
 ]
}
