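# Translucent unlit beam material for mesh particles, drawing distorted scrolling lines tinted by particle color.
# Two flow-texture samples distort three offset gradient samples for RGB fringing, boosted to HDR with a mobile fallback.
# World position offset combines streak-driven normal displacement, camera-distance thickening and a Grace-weighted pull toward EndPoint.
# Opacity combines axis-aligned fresnel, depth fade (skipped on mobile), UV end fades and near-camera fade.
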Material /Figment_S08/Athena/Prototype/Blueprints/Leaky/FX/M_Simple_Beam_Leaky
graph {
 properties {bUsedWithMeshParticles=true, blendMode=BLEND_Translucent, pixelDepthOffsetMode=PDOM_Legacy, shadingModel=MSM_Unlit, usedShadingModels=Unlit}
 nodes [
  n1 = VertexNormalWS()
  n2 = MaterialFunctionCall(Float2=TextureCoordinate()) {materialFunction=&BreakOutFloat2Components}
  n3 = Add(A=OneMinus(None=@n2.G), B=LinearInterpolate(Alpha=Time() {bOverride_Period=true, period=1.5}) {constA=-1, constB=0.4125}) {constB=0}
  n4 = Multiply(
   A=ScalarParameter() {group=WPO, parameterName=VertexOffset, sortPriority=1}
   B=TextureSample(UVs=AppendVector(A=@n2.R, B=@n3)) {texture=/Game/Effects/WeaponDestroy/T_WeaponDestroy_Streak}.R
  )
  n5 = Distance(A=ObjectPositionWS(), B=CameraPositionWS())
  n6 = ScalarParameter() {defaultValue=2000, parameterName=DistanceOffset}
  n7 = Clamp(None=Divide(A=Subtract(A=@n5, B=@n6), B=ScalarParameter() {defaultValue=3000, parameterName=ScaleUpDistance}))
  n8 = LinearInterpolate(B=Multiply(A=ScalarParameter() {defaultValue=100, parameterName=MaxScaleIncrease}, B=VertexNormalWS()) {constB=-1}, Alpha=@n7) {constB=0}
  n9 = SphereMask(A=WorldPosition().XYZ, B=CameraPositionWS(), Radius=ScalarParameter() {defaultValue=2500, parameterName=Radius}) {hardnessPercent=0}
  n10 = Subtract(A=VectorParameter() {parameterName=EndPoint, defaultValue=<rgba 67471.75 -43466.2109 3083.8064 0>}.RGB, B=WorldPosition().XYZ)
  n11 = Power(Base=ComponentMask(None=TextureCoordinate()) {g=true}, Exp=ScalarParameter() {defaultValue=1, parameterName=Contrast})
  Add(
   A=Add(A=Multiply(A=@n1, B=@n4), B=LinearInterpolate(A=@n8, Alpha=@n9) {constB=0})
   B=Multiply(A=Multiply(A=@n10, B=@n11), B=ScalarParameter() {parameterName=Grace})
  )
  Constant()
  n12 = Multiply(
   A=TextureCoordinate()
   B=Multiply(
    A=ScalarParameter() {defaultValue=25, parameterName=Base Tile}
    B=DynamicParameter() {defaultValue=<rgba 50 1 1 1>, paramNames [Lines Tile Scale, Param2, Param3, Param4]}.Lines Tile Scale
   )
  )
  n13 = MaterialFunctionCall(Float2=TextureCoordinate()) {materialFunction=&BreakOutFloat2Components}
  n14 = Multiply(A=@n13.G, B=DynamicParameter() {defaultValue=<rgba 50 1 1 1>, paramNames [Lines Tile Scale, Param2, Param3, Param4]}.Lines Tile Scale)
  n15 = Time()
  n16 = ScalarParameter() {defaultValue=0.05, parameterName=DistortA-Speed}
  n17 = Add(
   A=Multiply(A=AppendVector(A=@n13.R, B=@n14), B=ScalarParameter() {defaultValue=2, parameterName=DistortA-Tile})
   B=Multiply(A=Multiply(A=@n15, B=@n16), B=Constant2Vector() {g=0.25, r=1})
  )
  n18 = MaterialFunctionCall(Float2=TextureCoordinate()) {materialFunction=&BreakOutFloat2Components}
  n19 = AppendVector(
   A=@n18.R
   B=Multiply(A=@n18.G, B=DynamicParameter() {defaultValue=<rgba 50 1 1 1>, paramNames [Lines Tile Scale, Param2, Param3, Param4]}.Lines Tile Scale)
  )
  n20 = ScalarParameter() {defaultValue=4, parameterName=DistortB-Tile}
  n21 = Multiply(A=Time(), B=ScalarParameter() {defaultValue=-0.1, parameterName=DistortB-Speed})
  n22 = Constant2Vector() {g=1, r=0.25}
  n23 = Add(A=ScalarParameter() {defaultValue=0.33, parameterName=DistortB-Offset}, B=Add(A=Multiply(A=@n19, B=@n20), B=Multiply(A=@n21, B=@n22)))
  n24 = Add(
   A=ComponentMask(None=TextureSample(UVs=@n17) {samplerType=SAMPLERTYPE_LinearColor, texture=/Figment_S08/Environments/Sets/DigSite/Textures/T_Push_Flow}.RGB) {g=true, r=true}
   B=ComponentMask(None=TextureSample(UVs=@n23) {samplerType=SAMPLERTYPE_LinearColor, texture=&T_Push_Flow}.RGB) {g=true, r=true}
  )
  n25 = Multiply(A=ScalarParameter() {defaultValue=25, parameterName=Distort-BaseAmt}, B=Subtract(A=Normalize(VectorInput=@n24)) {constB=0.5})
  n26 = Add(A=@n12, B=@n25)
  n27 = Add(A=Multiply(A=@n25, B=ScalarParameter() {defaultValue=1.02, parameterName=Distort-GreenAmt}), B=@n12)
  n28 = Add(A=Multiply(A=@n25, B=ScalarParameter() {defaultValue=1.04, parameterName=Distort-BlueAmt}), B=@n12)
  n29 = MaterialFunctionCall(
   X=TextureSample(UVs=Panner(Coordinate=@n26) {speedY=-0.125}) {samplerType=SAMPLERTYPE_LinearColor, texture=/Game/Weapons/WeaponSkins/Textures/T_VertGradient}.R
   Y=TextureSample(UVs=Panner(Coordinate=@n27) {speedY=-0.125}) {samplerType=SAMPLERTYPE_LinearColor, texture=&T_VertGradient}.R
   Z=TextureSample(UVs=Panner(Coordinate=@n28) {speedY=-0.125}) {samplerType=SAMPLERTYPE_LinearColor, texture=&T_VertGradient}.R
  ) {materialFunction=&MakeFloat3}
  n30 = Saturate(None=TextureSample(UVs=@n17) {samplerType=SAMPLERTYPE_LinearColor, texture=/Figment_S08/Environments/Sets/DigSite/Textures/T_Push_Mask}.R)
  n31 = LinearInterpolate(
   A=ScalarParameter() {defaultValue=0.25, parameterName=LinesPower-Low}
   B=ScalarParameter() {defaultValue=10, parameterName=LinesPower}
   Alpha=@n30
  )
  n32 = Add(A=@n30, B=ScalarParameter() {defaultValue=0.05, parameterName=BubbleDarkAmt})
  n33 = Desaturation(
   None=Multiply(A=Power(Base=@n29.Result, Exp=@n31), B=Saturate(None=@n32))
   Fraction=ScalarParameter() {defaultValue=-2, parameterName=Desaturation}
  )
  n34 = ScalarParameter() {defaultValue=3, parameterName=Magnetic Intensity}
  n35 = Multiply(
   A=StaticSwitchParameter(True=Multiply(A=@n33, B=@n34), False=Constant() {r=1}) {defaultValue=true, parameterName=Add Chromatic Elements to Beam}
   B=ParticleColor().RGB
  )
  n36 = MaterialFunctionCall(LDR Input=Clamp(None=@n35), HDR Tint=@n35, HDR Intensity Multiplier=ScalarParameter() {defaultValue=1, parameterName=HDRIntensity}) {materialFunction=&DeriveHDRfromLDR}
  n37 = Multiply(A=@n35, B=ScalarParameter() {defaultValue=1, desc=FORT-1091755, parameterName=Lowend Boost})
  FeatureLevelSwitch(Default=@n36.RemappedRange, ES2=@n37, ES3_1=@n37, SM4_REMOVED=@n36.RemappedRange)
  MaterialFunctionCall() {materialFunction=&PlatformSwitch_1D}
  MaterialFunctionCall() {materialFunction=&PlatformSwitch_1D}
  n38 = ScalarParameter() {defaultValue=200, parameterName=Near Camera Distance}
  n39 = MaterialFunctionCall(
   Axis=Transform(None=Constant3Vector() {constant=<rgba 1 0 0 0>}) {transformSourceType=TRANSFORMSOURCE_ParticleWorld}
   Exponent=ScalarParameter() {defaultValue=5, parameterName=FresnelExp}
  ) {materialFunction=&AxisAlignedFresnel}
  n40 = MaterialFunctionCall(
   Axis=Transform(None=Constant3Vector() {constant=<rgba 1 0 0 0>}) {transformSourceType=TRANSFORMSOURCE_Local}
   Exponent=ScalarParameter() {defaultValue=5, parameterName=FresnelExp}
  ) {materialFunction=&AxisAlignedFresnel}
  n41 = FeatureLevelSwitch(Default=@n39.Fresnel, ES2=@n40.Fresnel, ES3_1=@n40.Fresnel, SM4_REMOVED=@n39.Fresnel)
  n42 = ScalarParameter() {parameterName=Grace}
  n43 = Multiply(
   A=ParticleColor().A
   B=Multiply(A=LinearInterpolate(A=@n41, Alpha=@n42), B=ScalarParameter() {defaultValue=0.25, parameterName=Fresnel Intensity}) {constA=1}
  )
  n44 = DepthFade(Opacity=@n43, FadeDistance=ScalarParameter() {defaultValue=500, parameterName=FadeDistance})
  n45 = FeatureLevelSwitch(Default=@n44, ES2=@n43, ES3_1=@n43, SM4_REMOVED=@n44, SM5=@n44, SM6=@n44)
  n46 = ComponentMask(None=TextureCoordinate()) {g=true}
  n47 = Power(Base=@n46, Exp=ScalarParameter() {defaultValue=1, parameterName=Begin Fade})
  n48 = Power(Base=OneMinus(None=@n46), Exp=LinearInterpolate(A=ScalarParameter() {defaultValue=1, parameterName=End Fade}, Alpha=@n7) {constB=3})
  n49 = Clamp(None=Multiply(A=Multiply(A=@n47, B=@n48), B=ScalarParameter() {defaultValue=100, parameterName=Fade Intensity}))
  Reroute(None=Multiply(
    A=MaterialFunctionCall(Distance=@n38) {materialFunction=&Near_Camera_Fade}.Result
    B=Multiply(A=@n45, B=@n49)
   ))
 ]
}
